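# Physics asset for the Islander Female pickaxe, a two-handed melee cosmetic weapon, defining bodies and joints for its dangling feathers.
# A long capsule covers the handle, and small capsules sit on the dyn_ feather bones.
# The front and top feathers attach directly to the handle, while three back feathers form a chain hanging from it.
# Angular limits with tight twist and modest swing restrict how far each feather moves.
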PhysicsAsset /BRCosmetics/Weapons/FORT_Melee/Pickaxe_Islander_Female/Meshes/Islander_Female_Physics
bodies {
 Melee_TwoHanded_Handle [{shapeType=Capsule, length=96.3335, radius=2.2064, center=<0.538 -1.6405 34.4508>}]
 dyn_feather_back_01 [{shapeType=Capsule, length=3, radius=1.5, center=<2 0 0>, rotation=<rot 90 0 0>}]
 dyn_feather_back_02 [{shapeType=Capsule, length=3, radius=1.5, center=<2 0 0>, rotation=<rot 90 0 0>}]
 dyn_feather_back_03 [{shapeType=Capsule, length=3, radius=1.5, center=<2 0 0>, rotation=<rot 90 0 0>}]
 dyn_feather_frnt [{shapeType=Capsule, length=3, radius=1.5, center=<3 0 0>, rotation=<rot 90 0 0>}]
 dyn_feather_top [{shapeType=Capsule, length=3, radius=1.5, center=<3 0 0>, rotation=<rot 90 0 0>}]
}
constraints [
 {bone1Name=dyn_feather_frnt, bone2Name=Melee_TwoHanded_Handle, swing1LimitDegrees=20, swing2LimitDegrees=15, twistLimitDegrees=5}
 {bone1Name=dyn_feather_back_01, bone2Name=Melee_TwoHanded_Handle, swing1LimitDegrees=30, swing2LimitDegrees=15, twistLimitDegrees=5}
 {bone1Name=dyn_feather_back_02, bone2Name=dyn_feather_back_01, swing1LimitDegrees=30, swing2LimitDegrees=15, twistLimitDegrees=5}
 {bone1Name=dyn_feather_back_03, bone2Name=dyn_feather_back_02, swing1LimitDegrees=30, swing2LimitDegrees=15, twistLimitDegrees=5}
 {bone1Name=dyn_feather_top, bone2Name=Melee_TwoHanded_Handle, swing1LimitDegrees=15, swing2LimitDegrees=10, twistLimitDegrees=5}
]
